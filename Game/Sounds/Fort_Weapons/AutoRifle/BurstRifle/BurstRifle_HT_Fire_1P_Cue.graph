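# First-person fire sound cue for the high-tier (HT) burst rifle, using first-person weapon attenuation and a fire sound class.
# A random node picks one of three close-fire waves, and a modulator varies pitch slightly upward and boosts volume.
# The quality-level node feeds the same random node to every tier, so all quality levels sound alike.
# A gunfire concurrency set is assigned, alongside owner-limited override values that stop the oldest voice.
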SoundCue /Game/Sounds/Fort_Weapons/AutoRifle/BurstRifle/BurstRifle_HT_Fire_1P_Cue
root=Modulator_2
nodes [
 {id=Modulator_2, children [QualityLevel_0], properties {pitchMax=1.1, pitchMin=1, volumeMax=2, volumeMin=1.8}}
 {id=QualityLevel_0, children [Random_2,Random_2,Random_2,Random_2]}
 {id=Random_2, children [WavePlayer_3,WavePlayer_4,WavePlayer_5], properties {weights [1,1,1]}}
 {id=WavePlayer_3, properties {soundWaveAssetPtr=&/BurstRifle_HT_Closefire_01}}
 {id=WavePlayer_4, properties {soundWaveAssetPtr=&/BurstRifle_HT_Closefire_02}}
 {id=WavePlayer_5, properties {soundWaveAssetPtr=&/BurstRifle_HT_Closefire_03}}
]
properties {
 attenuationSettings=&Fort_Weapon_Projectile_FirstPerson_Athena_Att
 bDebug=true
 maxDistance=2097152
 soundClassObject=&Fort_WeaponRanged_Fire_1P_PSM
 volumeMultiplier=0.7
 concurrencyOverrides {bLimitToOwner=true, maxCount=3, resolutionRule=StopOldest}
 concurrencySet [&Weapon_Gunfire_FirstPerson_Conc]
}
