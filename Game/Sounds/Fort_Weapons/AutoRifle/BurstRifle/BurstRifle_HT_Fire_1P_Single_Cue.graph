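# Sound cue for a single first-person close-range fire shot of the HT burst rifle in Fortnite.
# A random node picks one of six equally weighted close-fire single-shot waves, exempt from branch culling, and feeds a modulator that boosts volume and varies pitch slightly upward.
# It uses first-person weapon attenuation, the ranged-fire 1P sound class and a gunfire concurrency set.
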SoundCue /Game/Sounds/Fort_Weapons/AutoRifle/BurstRifle/BurstRifle_HT_Fire_1P_Single_Cue
root=Modulator_0
nodes [
 {id=Modulator_0, children [Random_0], properties {pitchMax=1.1, pitchMin=1, volumeMax=2, volumeMin=1.8}}
 {
  id=Random_0
  children [WavePlayer_1,WavePlayer_2,WavePlayer_3,WavePlayer_4,WavePlayer_5,WavePlayer_0]
  properties {bShouldExcludeFromBranchCulling=true, weights [1,1,1,1,1,1]}
 }
 {id=WavePlayer_0, properties {soundWaveAssetPtr=&/BurstRifle_HT_Closefire_06_Single}}
 {id=WavePlayer_1, properties {soundWaveAssetPtr=&/BurstRifle_HT_Closefire_01_Single}}
 {id=WavePlayer_2, properties {soundWaveAssetPtr=&/BurstRifle_HT_Closefire_02_Single}}
 {id=WavePlayer_3, properties {soundWaveAssetPtr=&/BurstRifle_HT_Closefire_03_Single}}
 {id=WavePlayer_4, properties {soundWaveAssetPtr=&/BurstRifle_HT_Closefire_04_Single}}
 {id=WavePlayer_5, properties {soundWaveAssetPtr=&/BurstRifle_HT_Closefire_05_Single}}
]
properties {
 attenuationSettings=&Fort_Weapon_Projectile_FirstPerson_Athena_Att
 bDebug=true
 maxDistance=2097152
 soundClassObject=&Fort_WeaponRanged_Fire_1P_PSM
 volumeMultiplier=0.85
 concurrencyOverrides {bLimitToOwner=true, maxCount=3, resolutionRule=StopOldest, volumeScale=0, volumeScaleAttackTime=0.5}
 concurrencySet [&Weapon_Gunfire_FirstPerson_Conc]
}
