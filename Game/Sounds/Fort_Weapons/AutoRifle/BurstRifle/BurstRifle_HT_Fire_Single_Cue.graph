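# Sound cue for a single BurstRifle HT gunshot, using third-person and distant ranged-weapon fire sound classes.
# A mixer layers close, medium and distant branches at descending levels, each with its own sound class node, attenuation and modulator.
# The close branch randomizes six waves, while the medium and distant random nodes each repeat one wave three times.
# Local and global gunfire concurrency sets govern overlapping shots.
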SoundCue /Game/Sounds/Fort_Weapons/AutoRifle/BurstRifle/BurstRifle_HT_Fire_Single_Cue
root=Mixer_0
nodes [
 {id=Attenuation_2, children [Modulator_2], properties {attenuationSettings=&Fort_Weapon_Projectile_Close_Athena_Att}}
 {
  id=Attenuation_3
  children [Modulator_3]
  properties {attenuationSettings=&Fort_Weapon_Projectile_Medium_Mono_Athena_Att}
 }
 {
  id=Attenuation_4
  children [Modulator_4]
  properties {attenuationSettings=&Fort_Weapon_Projectile_Distant_Mono_Athena_Att}
 }
 {id=Mixer_0, children [SoundClass_2,SoundClass_4,SoundClass_3], properties {inputVolume [0.8,0.6,0.25]}}
 {id=Modulator_2, children [Random_7], properties {pitchMax=1.1, pitchMin=1, volumeMax=1.7, volumeMin=1.5}}
 {id=Modulator_3, children [Random_5], properties {volumeMax=2, volumeMin=1.9}}
 {id=Modulator_4, children [Random_6], properties {volumeMax=1.95, volumeMin=1.85}}
 {id=Random_5, children [WavePlayer_6,WavePlayer_8,WavePlayer_9], properties {bShouldExcludeFromBranchCulling=true, weights [1,1,1]}}
 {id=Random_6, children [WavePlayer_11, WavePlayer_12, WavePlayer_13], properties {bShouldExcludeFromBranchCulling=true, weights [1,1,1]}}
 {
  id=Random_7
  children [WavePlayer_15, WavePlayer_21, WavePlayer_22, WavePlayer_23, WavePlayer_24, WavePlayer_14]
  properties {bShouldExcludeFromBranchCulling=true, weights [1,1,1,1,1,1]}
 }
 {
  id=SoundClass_2
  children [Attenuation_2]
  properties {soundClassOverride=&Fort_WeaponRanged_Fire_3P_PSM}
 }
 {
  id=SoundClass_3
  children [Attenuation_4]
  properties {soundClassOverride=&Fort_WeaponRanged_Fire_Distant}
 }
 {
  id=SoundClass_4
  children [Attenuation_3]
  properties {soundClassOverride=&Fort_WeaponRanged_Fire_Distant}
 }
 {id=WavePlayer_11, properties {soundWaveAssetPtr=&/BurstRifle_HT_Fire_Dist_03_Single}}
 {id=WavePlayer_12, properties {soundWaveAssetPtr=&BurstRifle_HT_Fire_Dist_03_Single}}
 {id=WavePlayer_13, properties {soundWaveAssetPtr=&BurstRifle_HT_Fire_Dist_03_Single}}
 {id=WavePlayer_14, properties {soundWaveAssetPtr=&/BurstRifle_HT_Closefire_06_Single}}
 {id=WavePlayer_15, properties {soundWaveAssetPtr=&/BurstRifle_HT_Closefire_01_Single}}
 {id=WavePlayer_21, properties {soundWaveAssetPtr=&/BurstRifle_HT_Closefire_02_Single}}
 {id=WavePlayer_22, properties {soundWaveAssetPtr=&/BurstRifle_HT_Closefire_03_Single}}
 {id=WavePlayer_23, properties {soundWaveAssetPtr=&/BurstRifle_HT_Closefire_04_Single}}
 {id=WavePlayer_24, properties {soundWaveAssetPtr=&/BurstRifle_HT_Closefire_05_Single}}
 {id=WavePlayer_6, properties {soundWaveAssetPtr=&/BurstRifle_HT_Fire_Med_01_Single}}
 {id=WavePlayer_8, properties {soundWaveAssetPtr=&BurstRifle_HT_Fire_Med_01_Single}}
 {id=WavePlayer_9, properties {soundWaveAssetPtr=&BurstRifle_HT_Fire_Med_01_Single}}
]
properties {
 bDebug=true
 maxDistance=27500
 soundClassObject=&Fort_WeaponRanged_Fire_3P_PSM
 volumeMultiplier=1
 concurrencyOverrides {bLimitToOwner=true, maxCount=3, resolutionRule=StopOldest}
 concurrencySet [&Weapon_Gunfire_Local_Conc, &Weapon_Gunfire_Global_Conc]
}
