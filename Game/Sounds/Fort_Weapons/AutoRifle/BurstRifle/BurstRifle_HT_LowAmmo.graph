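# Sound cue for the burst rifle's low-ammo sound, playing a battle rifle low-ammo fire SFX wave.
# A single wave player feeds a modulator that randomizes volume slightly above unity and keeps pitch fixed.
# It routes to the degraded low-ammo ranged weapon sound class and overrides concurrency so new instances are blocked once the limit is reached.
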SoundCue /Game/Sounds/Fort_Weapons/AutoRifle/BurstRifle/BurstRifle_HT_LowAmmo
root=Modulator_0
nodes [
 {id=Modulator_0, children [WavePlayer_2], properties {pitchMax=1, pitchMin=1, volumeMax=1.25, volumeMin=1.1}}
 {id=WavePlayer_2, properties {soundWaveAssetPtr=&/BATTLERIFLE_LOWAMMO_FIRE_SFX_v2}}
]
properties {
 bOverrideConcurrency=true
 maxDistance=2097152
 soundClassObject=/Game/Sounds/FortSoundClassesAndModes/Fort_WeaponRanged_LowAmmo_Degraded_PSM
 concurrencyOverrides {resolutionRule=PreventNew}
}
